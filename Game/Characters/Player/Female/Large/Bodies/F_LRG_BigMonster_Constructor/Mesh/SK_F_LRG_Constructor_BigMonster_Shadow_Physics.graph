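# Shadow physics asset for the large female BigMonster Constructor player body, its name pointing to use as a capsule shadow proxy rather than simulation.
# Capsules cover the head, clavicles, pelvis, spine and feet, with spine_02 through spine_04 each adding a second, rotated capsule.
# Arm, leg and neck_02 bodies list no shapes, and no constraints link any bodies.
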PhysicsAsset /Game/Characters/Player/Female/Large/Bodies/F_LRG_BigMonster_Constructor/Mesh/SK_F_LRG_Constructor_BigMonster_Shadow_Physics
bodies {
 calf_l []
 calf_r []
 clavicle_l [{shapeType=Capsule, length=7.6964, radius=9.8894, center=<4.0147 -0.8354 -1.7419>, rotation=<rot -77.5566 177.3915 7.7314>}]
 clavicle_r [{shapeType=Capsule, length=11.2269, radius=10.1744, center=<-8.067 0.5229 0.5316>, rotation=<rot -56.8981 7.4934 -16.5894>}]
 foot_l [{shapeType=Capsule, length=12.5982, radius=8.5402, center=<-5.1288 7.9378 -0.3592>, rotation=<rot 0.0388 1.4701 -89.3982>}]
 foot_r [{shapeType=Capsule, length=12.5982, radius=8.5402, center=<5.3225 -7.2143 0.3518>, rotation=<rot 0.0388 1.4701 90.6018>}]
 head [{shapeType=Capsule, length=0, radius=18, center=<9.0697 0.0746 0>, rotation=<rot 0 -90 0>}]
 lowerarm_l []
 lowerarm_r []
 neck_02 []
 pelvis [{shapeType=Capsule, length=33.7272, radius=12.0415, center=<-4.9708 -3.355 -1.1897>}]
 spine_01 [{shapeType=Capsule, length=14.5505, radius=13.3742, center=<1.4338 -1.0873 0>}]
 spine_02 [
  {shapeType=Capsule, length=7.4994, radius=11.3415, center=<1.179 2.2974 0>}
  {shapeType=Capsule, length=20.6875, radius=9.3125, center=<-1.8317 2.606 0>, rotation=<rot 90 0 0>}
 ]
 spine_03 [
  {shapeType=Capsule, length=7.6708, radius=9.9737, center=<2.9804 0.1017 0>}
  {shapeType=Capsule, length=20.7188, radius=9.2812, center=<-1.2326 2.6059 0>, rotation=<rot 90 0 0>}
 ]
 spine_04 [
  {shapeType=Capsule, length=5.3409, radius=14.1478, center=<5.5182 1.8272 0>}
  {shapeType=Capsule, length=10.0312, radius=9.2812, center=<-1.2326 1.308 0>, rotation=<rot 90 0 0>}
 ]
 spine_05 [{shapeType=Capsule, length=10.9859, radius=10.0791, center=<3.7401 0.3227 0>}]
 thigh_l []
 thigh_r []
 upperarm_l []
 upperarm_r []
}
constraints []
